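# Physics asset for the InspireSpell backpack mesh, defining capsule bodies for the bag and its four dangling dynamic bone chains.
# Six capsules approximate the bag, while the chain bones mostly carry one small capsule each.
# All four chains branch from the shared dyn_chain_CAR_Crep body through per-joint swing/twist constraints, some with locked axes, and link to mesh bodies at their ends.
# No constraint connects the bag to the chains.
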PhysicsAsset /BRCosmetics/Accessories/FORT_Backpacks/Backpack_M_MED_InspireSpell/Meshes/M_MED_InspireSpell_Pack_Physics
bodies {
 bag [
  {shapeType=Capsule, length=28.1732, radius=10.0416, shapeName=bag_capsule, center=<3.375 0.4951 0.4772>, rotation=<rot 0 -79.5367 -90.2357>}
  {shapeType=Capsule, length=10.2758, radius=4.673, shapeName=bag_capsule, center=<21.0819 5.0351 0.5718>}
  {shapeType=Capsule, length=9.5871, radius=3.7454, shapeName=bag_capsule, center=<22.6338 -1.167 0.908>}
  {
   shapeType=Capsule
   length=11.3501
   radius=3.496
   shapeName=bag_capsule
   center=<22.5014 2.8007 3.3017>
   rotation=<rot -0.0254 5.3869 90.2698>
  }
  {
   shapeType=Capsule
   length=7.5391
   radius=4.8668
   shapeName=bag_capsule
   center=<21.4959 0.2967 -2.2521>
   rotation=<rot -6.4073 6.5055 -45.6196>
  }
  {shapeType=Capsule, length=10.7002, radius=3.8495, shapeName=bag_capsule, center=<22.3979 3.1436 1.4352>}
 ]
 dyn_chain2_0 [{
   shapeType=Capsule
   length=2.5703
   radius=1.0076
   shapeName=dyn_chain2_0_capsule
   center=<2.0075 0.1327 -0.0349>
   rotation=<rot 0 92.7394 -87.219>
  }]
 dyn_chain_1_0 [{
   shapeType=Capsule
   length=2.6533
   radius=0.8791
   shapeName=dyn_chain_1_0_capsule
   center=<-1.5884 1.3153 0.3795>
   rotation=<rot 0 -131.1034 -79.4754>
  }]
 dyn_chain_1_10 [{
   shapeType=Capsule
   length=1.719
   radius=0.7487
   shapeName=dyn_chain_1_10_capsule
   center=<0.9054 0.0731 0.0009>
   rotation=<rot 0 95.3533 -87.5092>
  }]
 dyn_chain_1_11 [{shapeType=Capsule, length=0.505, radius=0.505, shapeName=dyn_chain_1_11_capsule, rotation=<rot 0 -90 0>}]
 dyn_chain_1_2 [{
   shapeType=Capsule
   length=2.6471
   radius=0.9033
   shapeName=dyn_chain_1_2_capsule
   center=<1.8988 -0.0639 -0.0124>
   rotation=<rot 0 -93.3832 -89.2221>
  }]
 dyn_chain_1_4 [{
   shapeType=Capsule
   length=2.6734
   radius=0.9552
   shapeName=dyn_chain_1_4_capsule
   center=<1.8612 -0.2112 0.2201>
   rotation=<rot 0 91.9933 -75.7212>
  }]
 dyn_chain_1_6 [{
   shapeType=Capsule
   length=2.6307
   radius=1.0843
   shapeName=dyn_chain_1_6_capsule
   center=<1.7156 0.163 0.258>
   rotation=<rot 0 86.5549 -76.353>
  }]
 dyn_chain_1_8 [{
   shapeType=Capsule
   length=2.9183
   radius=0.9125
   shapeName=dyn_chain_1_8_capsule
   center=<1.7144 0.3306 -0.0012>
   rotation=<rot 0 -62.5207 -93.6882>
  }]
 dyn_chain_1_mesh_2 [{
   shapeType=Capsule
   length=3.2352
   radius=0.7962
   shapeName=dyn_chain_1_mesh_2_capsule
   center=<1.149 0.0143 0.2116>
   rotation=<rot 0 -177.8706 -87.0205>
  }]
 dyn_chain_2_10 [{
   shapeType=Capsule
   length=2.6225
   radius=1.0793
   shapeName=dyn_chain_2_10_capsule
   center=<1.8132 0.1327 0.0401>
   rotation=<rot 0 -87.5806 -75.5935>
  }]
 dyn_chain_2_2 [{
   shapeType=Capsule
   length=2.4535
   radius=1.1179
   shapeName=dyn_chain_2_2_capsule
   center=<1.7444 -0.1931 -0.2066>
   rotation=<rot 0 94.165 -88.6867>
  }]
 dyn_chain_2_4 [{
   shapeType=Capsule
   length=3.2548
   radius=0.5349
   shapeName=dyn_chain_2_4_capsule
   center=<1.4039 -0.11 0.1528>
   rotation=<rot 0 -86.7184 -89.3664>
  }]
 dyn_chain_2_6 [{
   shapeType=Capsule
   length=2.5026
   radius=1.2697
   shapeName=dyn_chain_2_6_capsule
   center=<1.8719 -0.216 -0.007>
   rotation=<rot 0 -86.4128 -87.9016>
  }]
 dyn_chain_2_8 [{
   shapeType=Capsule
   length=2.3551
   radius=1.2956
   shapeName=dyn_chain_2_8_capsule
   center=<1.9701 -0.1219 -0.0377>
   rotation=<rot 0 95.7794 -87.3099>
  }]
 dyn_chain_2_mesh_2 [{
   shapeType=Capsule
   length=3.2329
   radius=0.7725
   shapeName=dyn_chain_2_mesh_2_capsule
   center=<1.1581 0.0252 0.1033>
   rotation=<rot 0 1.3462 -89.7366>
  }]
 dyn_chain_3_0 [{
   shapeType=Capsule
   length=2.9513
   radius=0.8242
   shapeName=dyn_chain_3_0_capsule
   center=<1.6961 -0.1575 -0.0234>
   rotation=<rot 0 -89.2228 -77.7242>
  }]
 dyn_chain_3_2 [{
   shapeType=Capsule
   length=2.65
   radius=0.7727
   shapeName=dyn_chain_3_2_capsule
   center=<1.9122 -0.0774 -0.1169>
   rotation=<rot 0 -91.083 -85.0721>
  }]
 dyn_chain_3_4 [{
   shapeType=Capsule
   length=2.8605
   radius=0.5738
   shapeName=dyn_chain_3_4_capsule
   center=<1.768 0.1433 -0.0947>
   rotation=<rot 0.0027 91.2242 -77.9853>
  }]
 dyn_chain_3_6 [{
   shapeType=Capsule
   length=2.4458
   radius=0.9722
   shapeName=dyn_chain_3_6_capsule
   center=<2.3025 0.1732 -0.0583>
   rotation=<rot 0 82.8069 -88.7967>
  }]
 dyn_chain_3_8 [{
   shapeType=Capsule
   length=2.5168
   radius=1.0676
   shapeName=dyn_chain_3_8_capsule
   center=<1.4272 -0.112 -0.3576>
   rotation=<rot 0 83.9577 -81.7016>
  }]
 dyn_chain_3mesh_2 [{
   shapeType=Capsule
   length=2.733
   radius=1.2758
   shapeName=dyn_chain_3mesh_2_capsule
   center=<1.7939 -0.0024 -0.6529>
   rotation=<rot 72.7507 0.0118 87.7086>
  }]
 dyn_chain_4_0 [{
   shapeType=Capsule
   length=2.7409
   radius=0.9897
   shapeName=dyn_chain_4_0_capsule
   center=<1.9386 -0.1289 -0.0425>
   rotation=<rot 0 86.9328 -88.7504>
  }]
 dyn_chain_4_2 [{
   shapeType=Capsule
   length=2.5922
   radius=0.8882
   shapeName=dyn_chain_4_2_capsule
   center=<1.8465 -0.0366 -0.2023>
   rotation=<rot 0 90.226 -87.2398>
  }]
 dyn_chain_4_4 [{
   shapeType=Capsule
   length=3.0194
   radius=0.7898
   shapeName=dyn_chain_4_4_capsule
   center=<1.9838 0.1993 0.0117>
   rotation=<rot 0 94.1595 -90.4877>
  }]
 dyn_chain_4_6 [{
   shapeType=Capsule
   length=2.5556
   radius=0.9584
   shapeName=dyn_chain_4_6_capsule
   center=<1.5162 0.3237 -0.0931>
   rotation=<rot 0 83.0541 -84.3621>
  }]
 dyn_chain_4_8 [{shapeType=Capsule, length=0.505, radius=0.505, shapeName=dyn_chain_4_8_capsule, rotation=<rot 0 -90 0>}]
 dyn_chain_4_mesh_1 [{shapeType=Capsule, length=0.505, radius=0.505, shapeName=dyn_chain_4_mesh_1_capsule, rotation=<rot 0 -90 -40.332>}]
 dyn_chain_4_mesh_2 [{
   shapeType=Capsule
   length=2.9909
   radius=0.9292
   shapeName=dyn_chain_4_mesh_2_capsule
   center=<1.2338 0.1417 -0.1163>
   rotation=<rot 0 10.1665 -85.3284>
  }]
 dyn_chain_CAR_Crep [{
   shapeType=Capsule
   length=2.3763
   radius=1.4426
   shapeName=dyn_chain_CAR_Crep_capsule
   center=<1.1056 0.066 -0.0624>
   rotation=<rot 21.6471 -90 86.847>
  }]
}
constraints [
 {bone1Name=dyn_chain_1_2, bone2Name=dyn_chain_1_0, swing1LimitDegrees=15, swing2LimitDegrees=50, twistLimitDegrees=15}
 {bone1Name=dyn_chain_1_4, bone2Name=dyn_chain_1_2, swing1LimitDegrees=21, swing2LimitDegrees=50, twistLimitDegrees=15}
 {bone1Name=dyn_chain_1_6, bone2Name=dyn_chain_1_4, swing1LimitDegrees=20, swing2LimitDegrees=50, twistLimitDegrees=15}
 {bone1Name=dyn_chain_1_8, bone2Name=dyn_chain_1_6, swing1LimitDegrees=20, swing2LimitDegrees=50, twistLimitDegrees=15}
 {bone1Name=dyn_chain_1_10, bone2Name=dyn_chain_1_8, swing1LimitDegrees=30, swing2LimitDegrees=30, twistLimitDegrees=3}
 {bone1Name=dyn_chain_2_2, bone2Name=dyn_chain2_0, swing1LimitDegrees=50, swing2LimitDegrees=50, twistLimitDegrees=15}
 {bone1Name=dyn_chain_2_4, bone2Name=dyn_chain_2_2, swing1LimitDegrees=50, swing2LimitDegrees=50, twistLimitDegrees=15}
 {bone1Name=dyn_chain_3_2, bone2Name=dyn_chain_3_0, swing1LimitDegrees=15, swing2LimitDegrees=50, twistLimitDegrees=5}
 {bone1Name=dyn_chain_3_4, bone2Name=dyn_chain_3_2, swing1LimitDegrees=15, swing2LimitDegrees=50, twistLimitDegrees=15}
 {bone1Name=dyn_chain_3_6, bone2Name=dyn_chain_3_4, swing1LimitDegrees=30, swing2LimitDegrees=50, twistLimitDegrees=5}
 {bone1Name=dyn_chain_3_8, bone2Name=dyn_chain_3_6, swing1LimitDegrees=20, swing2LimitDegrees=20, twistLimitDegrees=5}
 {bone1Name=dyn_chain_4_2, bone2Name=dyn_chain_4_0, swing1LimitDegrees=50, swing2LimitDegrees=50, twistLimitDegrees=15}
 {bone1Name=dyn_chain_4_4, bone2Name=dyn_chain_4_2, swing1LimitDegrees=50, swing2LimitDegrees=50, twistLimitDegrees=15}
 {bone1Name=dyn_chain_4_6, bone2Name=dyn_chain_4_4, swing1LimitDegrees=50, swing2LimitDegrees=50, twistLimitDegrees=15}
 {bone1Name=dyn_chain_3_0, bone2Name=dyn_chain_CAR_Crep, swing1LimitDegrees=5, swing2LimitDegrees=35, twistLimitDegrees=3}
 {bone1Name=dyn_chain_4_0, bone2Name=dyn_chain_CAR_Crep, swing1LimitDegrees=30, swing2LimitDegrees=50, twistLimitDegrees=7}
 {bone1Name=dyn_chain2_0, bone2Name=dyn_chain_CAR_Crep, swing1LimitDegrees=10, swing2LimitDegrees=10, twistLimitDegrees=5}
 {
  bone1Name=dyn_chain_1_0
  bone2Name=dyn_chain_CAR_Crep
  swing1LimitDegrees=35
  swing1Motion=Locked
  swing2LimitDegrees=35
  twistLimitDegrees=7
  twistMotion=Locked
 }
 {bone1Name=dyn_chain_2_6, bone2Name=dyn_chain_2_4, swing1LimitDegrees=50, swing2LimitDegrees=50, twistLimitDegrees=15}
 {bone1Name=dyn_chain_2_8, bone2Name=dyn_chain_2_6, swing1LimitDegrees=50, swing2LimitDegrees=50, twistLimitDegrees=15}
 {bone1Name=dyn_chain_2_10, bone2Name=dyn_chain_2_8, swing1LimitDegrees=30, swing2LimitDegrees=30, twistLimitDegrees=15}
 {
  bone1Name=dyn_chain_1_mesh_2
  bone2Name=dyn_chain_1_10
  swing1LimitDegrees=40
  swing2LimitDegrees=40
  twistLimitDegrees=45
  twistMotion=Locked
 }
 {bone1Name=dyn_chain_3_8, bone2Name=dyn_chain_3mesh_2, swing1LimitDegrees=45, swing2LimitDegrees=45, twistLimitDegrees=3}
 {bone1Name=dyn_chain_4_8, bone2Name=dyn_chain_4_6, swing1LimitDegrees=45, swing2LimitDegrees=45, twistLimitDegrees=45, twistMotion=Locked}
 {bone1Name=dyn_chain_4_8, bone2Name=dyn_chain_4_mesh_2, swing1LimitDegrees=45, swing2LimitDegrees=45, twistLimitDegrees=3}
 {
  bone1Name=dyn_chain_2_10
  bone2Name=dyn_chain_2_mesh_2
  swing1LimitDegrees=45
  swing2LimitDegrees=45
  twistLimitDegrees=45
  twistMotion=Locked
 }
]
