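# Animation Blueprint for the InspireSpell backpack cosmetic, adding simulated physics and spine-driven tilt.
# A RigidBody node simulates relative to the bag bone, with gravity and velocity/acceleration scales switched each update for front-end, skydiving or default states.
# Two root bone modifiers follow: a fixed offset, then an additive roll mapped from the owner mesh's spine yaw.
# The computed Backpack Translate feeds that modifier's ignored Scale input, not Translation.
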AnimBlueprint /BRCosmetics/Accessories/FORT_Backpacks/Backpack_M_MED_InspireSpell/Meshes/M_MED_InspireSpell_Pack_AnimBP
target_skeleton=&/M_MED_InspireSpell_Pack_Skeleton
graphs [
 AnimGraph {nodes [
   n1 = RigidBody!(
    ComponentPose=Local To Component(LocalPose=Input Pose())
    Alpha
    OverrideWorldGravity:Vector=Get gravity() -> (gravity:Vector)
    ExternalForce
    ComponentLinearAccScale:Vector=Get Linear acc scale() -> (Linear acc scale:Vector)
    ComponentLinearVelScale:Vector=Get Linear vel scale() -> (Linear vel scale:Vector)
   ) -> (Pose) {
    bClampLinearTranslationLimitToRefPose=true
    bOverrideWorldGravity=true
    overridePhysicsAsset=&/M_MED_InspireSpell_Pack_Physics
    simulationSpace=BaseBoneSpace
    alphaBoolBlend {blendOption=Linear}
    baseBoneRef {boneName=bag}
    componentAppliedLinearAccClamp=<1000 1000 1000>
    componentLinearAccScale=<0.6 0.6 0.6>
    componentLinearVelScale=<0.6 0.6 0.6>
    overrideWorldGravity {z=-900}
    simSpaceSettings {
     maxAngularAcceleration=500
     maxAngularVelocity=3
     maxLinearAcceleration=400
     maxLinearVelocity=0
     velocityScaleZ=0
     externalLinearDragV=<0.1 0.1>
    }
   }
   n2 = "Transform (Modify) Bone - Bone: root"(ComponentPose=@n1, Translation="0,-1,0", Rotation="0,0,2") {
    rotationMode=BMM_Replace
    rotationSpace=BCS_BoneSpace
    translationMode=BMM_Replace
    translationSpace=BCS_BoneSpace
    alphaBoolBlend {blendOption=Linear}
    boneToModify {boneName=root}
    rotation {roll=2}
    translation {y=-1}
   }
   n3 = "Transform (Modify) Bone - Bone: root"(
    ComponentPose=@n2
    Translation="0,0,0"
    Rotation=Get Backpack Rotate() -> (Backpack Rotate:Rotator)
    Scale=Get Backpack Translate() -> (Backpack Translate:Vector)
   ) {
    rotationMode=BMM_Additive
    rotationSpace=BCS_BoneSpace
    translationMode=BMM_Additive
    translationSpace=BCS_BoneSpace
    alphaBoolBlend {blendOption=Linear}
    boneToModify {boneName=root}
   }
   Output Pose(Result=Component To Local(ComponentPose=@n3)) {layerGroup=None}
  ]}
 EventGraph {nodes [
   Event BlueprintUpdateAnimation()
   > n1 = Sequence()
   > n2 = Branch(Condition=Get bIsInFrontEnd())
   > n3 = Sequence!(execute) -> (then_0, then_1, then_2:Exec)
   > Set Linear vel scale(Linear vel scale:Vector="0.3,0.3,0.3") -> (Output_Get:Vector)
   Set Linear acc scale(execute=@n3.then_1, Linear acc scale:Vector="0.3,0.3,0.3") -> (Output_Get:Vector)
   Set gravity(execute=@n3.then_2, gravity:Vector="0,0,-920") -> (Output_Get:Vector)
   n4 = Branch(execute=@n2.else, Condition=Get bIsSkyDiving())
   > n5 = Sequence()
   > Set Linear vel scale(Linear vel scale="0.8,0.8,0.8")
   Set Linear acc scale(execute=@n5.then_1, Linear acc scale="0.8,0.8,0.8")
   Set gravity(execute=@n5.then_2, gravity="0,0,-10")
   n6 = Sequence(execute=@n4.else)
   > Set Linear acc scale(Linear acc scale="0.55,0.55,0.55")
   Set Linear vel scale(execute=@n6.then_1, Linear vel scale="0.55,0.55,0.55")
   Set gravity(execute=@n6.then_2, gravity="0,0,-870")
   n7 = GetOwnerSkeletalMeshComponent()
   n8 = BreakRotator(InRot=BreakTransform(InTransform=GetDeltaTransformFromRefPose(self=@n7, BoneName=spine_05, BaseName=spine_01)).Rotation)
   Set Backpack Rotate(execute=@n1.then_1, Backpack Rotate:Rotator=MakeRotator(Roll=MapRangeClamped(Value=@n8.Yaw, InRangeA=0, InRangeB=-30, OutRangeB=35))) -> (Output_Get:Rotator)
   > Set Backpack Translate(Backpack Translate:Vector=MakeVector(Y=MapRangeClamped(Value=@n8.Yaw, InRangeA=0, InRangeB=-30, OutRangeB=-10))) -> (Output_Get:Vector)
  ]}
]
variables [
 Backpack Rotate {type=Rotator}
 Backpack Translate {type=Vector}
 carabin {type=Vector}
 gravity {type=Vector}
 Linear acc scale {type=Vector}
 Linear vel scale {type=Vector}
]
